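# Looping thrust sound cue for the JollyTroll cosmetic glider.
# A mixer blends five looping layers (flap, gear, jingle, sequins and wind) at per-layer levels.
# Each layer has its own continuous modulator mapping the AnyDirection parameter to volume and pitch ranges, so the loop responds to movement input.
# Shared glider thrust attenuation, glider sound class and cosmetic loop properties sheet govern playback.
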SoundCue /BRCosmetics/Sounds/Gliders/JollyTroll/GliderThrustLoop_JollyTroll_Cue
root=Mixer_0
nodes [
 {
  id=Mixer_0
  children [ModulatorContinuous_0, ModulatorContinuous_1, ModulatorContinuous_2, ModulatorContinuous_3, ModulatorContinuous_4]
  properties {inputVolume [1,1,0.5,0.8,0.7]}
 }
 {
  id=ModulatorContinuous_0
  children [WavePlayer_6]
  properties {
   pitchModulationParams {default=0, maxOutput=1.1, minOutput=1, parameterName=AnyDirection}
   volumeModulationParams {default=0, minOutput=0.5, parameterName=AnyDirection}
  }
 }
 {
  id=ModulatorContinuous_1
  children [WavePlayer_7]
  properties {
   pitchModulationParams {default=0, maxOutput=1.1, minOutput=1, parameterName=AnyDirection}
   volumeModulationParams {default=0, maxOutput=0.8, minOutput=0.5, parameterName=AnyDirection}
  }
 }
 {
  id=ModulatorContinuous_2
  children [WavePlayer_8]
  properties {
   pitchModulationParams {default=0, minOutput=1, parameterName=AnyDirection}
   volumeModulationParams {default=0, maxOutput=0.8, minOutput=0.4, parameterName=AnyDirection}
  }
 }
 {
  id=ModulatorContinuous_3
  children [WavePlayer_9]
  properties {
   pitchModulationParams {default=0, maxOutput=1.2, minOutput=1, parameterName=AnyDirection}
   volumeModulationParams {default=0, minOutput=0.6, parameterName=AnyDirection}
  }
 }
 {
  id=ModulatorContinuous_4
  children [WavePlayer_10]
  properties {
   pitchModulationParams {default=0, maxOutput=1.2, minOutput=1, parameterName=AnyDirection}
   volumeModulationParams {default=0, maxOutput=0.8, minOutput=0.4, parameterName=AnyDirection}
  }
 }
 {id=WavePlayer_10, properties {bLooping=true, soundWaveAssetPtr=&/Glider_JollyTroll_Thrust_LP_Wind}}
 {id=WavePlayer_6, properties {bLooping=true, soundWaveAssetPtr=&/Glider_JollyTroll_Thrust_LP_Flap}}
 {id=WavePlayer_7, properties {bLooping=true, soundWaveAssetPtr=&/Glider_JollyTroll_Thrust_LP_Gear}}
 {id=WavePlayer_8, properties {bLooping=true, soundWaveAssetPtr=&/Glider_JollyTroll_Thrust_LP_Jingle}}
 {id=WavePlayer_9, properties {bLooping=true, soundWaveAssetPtr=&/Glider_JollyTroll_Thrust_LP_Sequins}}
]
properties {
 attenuationSettings=&Fort_Player_Glider_Thrust_Athena_Att
 audioPropertiesSheet=&APS_Cosmetic_Glider_Loop
 maxDistance=12500
 soundClassObject=&Fort_Gliders_Land
 volumeMultiplier=0.9
 concurrencySet [null]
}
